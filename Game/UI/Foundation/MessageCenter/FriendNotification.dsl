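; FriendNotification is a Blueprint asset in the Message Center UI, apparently for friend and party invite notifications.
; It declares two functions, JoinPartyInvite and ShowFriendInvites, whose names point to joining a party from an invite and displaying pending friend invites.
; Both functions show no body, parameters or connections, so the file holds no visible logic.
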(fn JoinPartyInvite)

(fn ShowFriendInvites)
